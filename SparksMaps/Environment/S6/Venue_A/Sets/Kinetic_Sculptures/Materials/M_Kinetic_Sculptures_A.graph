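# Lit material for the Venue A kinetic sculptures, using shared stage gold textures and a color parameter, enabled for instanced static meshes.
# Time drives a sine oscillation, with speed and amount parameters, that rotates local positions around an axis parameter for shader-based motion.
# The vertex normal gets the same rotation and is blended with the gold normal map, with tangent-space normals disabled.
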Material /SparksMaps/Environment/S6/Venue_A/Sets/Kinetic_Sculptures/Materials/M_Kinetic_Sculptures_A
graph {
 properties {bAutomaticallySetUsageInEditor=false, bTangentSpaceNormal=false, bUsedWithInstancedStaticMeshes=true, usedShadingModels=Default Lit}
 nodes [
  n1 = VertexNormalWS()
  n2 = RotateAboutAxis(
   NormalizedRotationAxis=Transform(None=NamedRerouteUsage()) {transformSourceType=TRANSFORMSOURCE_Local}
   RotationAngle=NamedRerouteUsage()
   PivotPoint=Constant()
   Position=@n1
  )
  MaterialFunctionCall(
   BaseNormal=TextureSample() {samplerType=SAMPLERTYPE_Normal, texture=/SparksMaps/Environment/Shared/Materials/Generic/Textures/T_Stage_Gold_A_N}.RGB
   AdditionalNormal=Normalize(VectorInput=Add(A=@n2, B=@n1))
  ) {materialFunction=&BlendAngleCorrectedNormals}
  NamedRerouteDeclaration(None=Normalize(VectorInput=VectorParameter() {parameterName=Rotation Axis, defaultValue=<rgba 0 1 0.5 0>}.RGB)) {name=RotationAxis, nodeColor=<rgba 1 0 0.6797 1>}
  n3 = Multiply(A=Time(), B=ScalarParameter() {defaultValue=0.02, parameterName=Rotation Speed})
  NamedRerouteDeclaration(None=Multiply(A=Sine(None=@n3), B=ScalarParameter() {defaultValue=1, parameterName=Rotation Amount})) {name=RotationAngle, nodeColor=<rgba 0 1 1 1>}
  TextureSample() {samplerType=SAMPLERTYPE_Masks, texture=/SparksMaps/Environment/Shared/Materials/Generic/Textures/T_Stage_Gold_A_S}
  TextureSample() {texture=/SparksMaps/Environment/Shared/Materials/Generic/Textures/T_Stage_Gold_A_D}
  Transform(None=RotateAboutAxis(NormalizedRotationAxis=NamedRerouteUsage(), RotationAngle=NamedRerouteUsage(), PivotPoint=Constant(), Position=LocalPosition().XYZ)) {transformSourceType=TRANSFORMSOURCE_Local}
  VectorParameter() {parameterName=Color, defaultValue=<rgba 1 0.5683 0.2174 1>}
 ]
}
